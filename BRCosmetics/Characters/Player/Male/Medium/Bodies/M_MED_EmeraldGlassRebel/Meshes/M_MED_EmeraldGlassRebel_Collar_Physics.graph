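# Physics asset for the collar of the Emerald Glass Rebel medium male player body cosmetic, built for secondary motion of collar bones.
# Capsules and boxes on upper spine, neck, head and clavicles form the body shapes around front, back, left and right collar flaps.
# Each flap capsule bone is constrained with tight angular limits to a small MSTR anchor, which links to spine_05.
# Side flaps swing on one axis, front and back on two, all without twist.
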PhysicsAsset /BRCosmetics/Characters/Player/Male/Medium/Bodies/M_MED_EmeraldGlassRebel/Meshes/M_MED_EmeraldGlassRebel_Collar_Physics
bodies {
 clavicle_l [
  {shapeType=Capsule, length=3, radius=3, shapeName=clavicle_l_capsule}
  {
   shapeType=Box
   extentX=3.6222
   extentY=19.3958
   extentZ=23.295
   shapeName=clavicle_l_box
   center=<11.9117 -0.8856 6.4768>
   rotation=<rot -8.4883 -11.9508 15.1557>
  }
 ]
 clavicle_r [
  {shapeType=Capsule, length=3, radius=3, shapeName=clavicle_l_capsule, rotation=<rot 0 0 -180>}
  {
   shapeType=Box
   extentX=3.6222
   extentY=19.3958
   extentZ=23.295
   shapeName=clavicle_l_box
   center=<-11.9117 0.8856 -6.4768>
   rotation=<rot -8.4883 -11.9508 -164.8443>
  }
 ]
 dyn_collarBk_c_01 [{shapeType=Capsule, length=6.8077, radius=2.3898, shapeName=dyn_collarBk_c_01_capsule, center=<-5.1417 -3 0>, rotation=<rot 90 0 0>}]
 dyn_collarBk_c_MSTR [{shapeType=Capsule, length=0.5, radius=0.5, shapeName=dyn_collarBk_c_MSTR_capsule}]
 dyn_collarFr_c_01 [{shapeType=Capsule, length=9.1662, radius=2.3898, shapeName=dyn_collarFr_c_01_capsule, center=<-9.0366 0 0>, rotation=<rot 2.5033 0 0>}]
 dyn_collarFr_c_MSTR [{shapeType=Capsule, length=0.5, radius=0.5, shapeName=dyn_collarFr_c_MSTR_capsule}]
 dyn_collar_l_01 [
  {
   shapeType=Capsule
   length=11.3909
   radius=1.8214
   shapeName=dyn_collar_l_01_capsule
   center=<-7.583 4.0997 1.4>
   rotation=<rot 90 -165.9638 -75.8752>
  }
  {
   shapeType=Capsule
   length=9.5912
   radius=1.5389
   shapeName=dyn_collar_l_01_capsule
   center=<-7.3529 -4.4669 -1.0383>
   rotation=<rot 2.8525 4.44 57.3436>
  }
 ]
 dyn_collar_l_MSTR [{shapeType=Capsule, length=0.5, radius=0.5, shapeName=dyn_collar_l_MSTR_capsule}]
 dyn_collar_r_01 [
  {
   shapeType=Capsule
   length=10.345
   radius=1.3469
   shapeName=dyn_collar_r_01_capsule
   center=<-7.65 4.5223 -1.3224>
   rotation=<rot 87.2277 89.6068 179.6601>
  }
  {
   shapeType=Capsule
   length=11.2762
   radius=1.4332
   shapeName=dyn_collar_r_01_capsule
   center=<-8.0149 -3.8955 1.0729>
   rotation=<rot -0.8541 1.4612 -59.6991>
  }
 ]
 dyn_collar_r_MSTR [{shapeType=Capsule, length=0.5, radius=0.5, shapeName=dyn_collar_r_MSTR_capsule}]
 head [
  {shapeType=Capsule, length=11.018, radius=8.1133, shapeName=head_capsule, center=<1.3327 5.2229 0>, rotation=<rot 90 -176.4237 164.7297>}
  {shapeType=Capsule, length=6.7384, radius=8.2324, shapeName=head_capsule, center=<5.0914 1.5645 0>, rotation=<rot 90 -2.3859 10.0432>}
 ]
 neck_01 [{shapeType=Capsule, length=1.6099, radius=6.5943, shapeName=neck_01_capsule, center=<0 1.1871 0>, rotation=<rot 90 0 0>}]
 neck_02 [{shapeType=Capsule, length=1.2311, radius=6.9731, shapeName=neck_02_capsule, center=<0 1.1871 0>, rotation=<rot 90 0 0>}]
 spine_05 [{shapeType=Capsule, length=19.7724, radius=11.5766, center=<-4.0377 0 0>}]
}
constraints [
 {
  bone1Name=clavicle_l
  bone2Name=spine_05
  swing1LimitDegrees=45
  swing1Motion=Free
  swing2LimitDegrees=45
  swing2Motion=Free
  twistLimitDegrees=45
  twistMotion=Free
 }
 {
  bone1Name=neck_01
  bone2Name=spine_05
  swing1LimitDegrees=45
  swing1Motion=Free
  swing2LimitDegrees=45
  swing2Motion=Free
  twistLimitDegrees=45
  twistMotion=Free
 }
 {
  bone1Name=neck_02
  bone2Name=neck_01
  swing1LimitDegrees=45
  swing1Motion=Free
  swing2LimitDegrees=45
  swing2Motion=Free
  twistLimitDegrees=45
  twistMotion=Free
 }
 {
  bone1Name=head
  bone2Name=neck_02
  swing1LimitDegrees=45
  swing1Motion=Free
  swing2LimitDegrees=45
  swing2Motion=Free
  twistLimitDegrees=45
  twistMotion=Free
 }
 {bone1Name=dyn_collar_r_01, bone2Name=dyn_collar_r_MSTR, swing1LimitDegrees=0, swing2LimitDegrees=15, twistLimitDegrees=0}
 {
  bone1Name=dyn_collar_r_MSTR
  bone2Name=spine_05
  swing1LimitDegrees=45
  swing1Motion=Free
  swing2LimitDegrees=45
  swing2Motion=Free
  twistLimitDegrees=45
  twistMotion=Free
 }
 {
  bone1Name=dyn_collar_l_MSTR
  bone2Name=spine_05
  swing1LimitDegrees=45
  swing1Motion=Free
  swing2LimitDegrees=45
  swing2Motion=Free
  twistLimitDegrees=45
  twistMotion=Free
 }
 {bone1Name=dyn_collar_l_01, bone2Name=dyn_collar_l_MSTR, swing1LimitDegrees=0, swing2LimitDegrees=15, twistLimitDegrees=0}
 {bone1Name=dyn_collarFr_c_01, bone2Name=dyn_collarFr_c_MSTR, swing1LimitDegrees=15, swing2LimitDegrees=15, twistLimitDegrees=0}
 {
  bone1Name=dyn_collarFr_c_MSTR
  bone2Name=spine_05
  swing1LimitDegrees=45
  swing1Motion=Free
  swing2LimitDegrees=45
  swing2Motion=Free
  twistLimitDegrees=45
  twistMotion=Free
 }
 {bone1Name=dyn_collarBk_c_01, bone2Name=dyn_collarBk_c_MSTR, swing1LimitDegrees=15, swing2LimitDegrees=15, twistLimitDegrees=0}
 {
  bone1Name=dyn_collarBk_c_MSTR
  bone2Name=spine_05
  swing1LimitDegrees=45
  swing1Motion=Free
  swing2LimitDegrees=45
  swing2Motion=Free
  twistLimitDegrees=45
  twistMotion=Free
 }
 {
  bone1Name=clavicle_r
  bone2Name=spine_05
  swing1LimitDegrees=45
  swing1Motion=Free
  swing2LimitDegrees=45
  swing2Motion=Free
  twistLimitDegrees=45
  twistMotion=Free
 }
]
