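# Physics asset for the dangling grenades on the EmeraldGlassRebel medium male body cosmetic, covering front and back grenades on left and right.
# Each grenade is a two-bone chain: a tiny MSTR capsule joins the pelvis capsule, and an offset capsule on the 01 bone forms the grenade body.
# MSTR-to-pelvis joints have free swing and twist, while grenade-to-MSTR joints have tight angular limits for restrained jiggle.
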PhysicsAsset /BRCosmetics/Characters/Player/Male/Medium/Bodies/M_MED_EmeraldGlassRebel/Meshes/M_MED_EmeraldGlassRebel_Grenades_Physics
bodies {
 dyn_grenadeBk_l_01 [{shapeType=Capsule, length=3, radius=3, shapeName=dyn_grenadeBk_l_01_capsule, center=<0 0 3.6>, rotation=<rot 90 0 0>}]
 dyn_grenadeBk_l_MSTR [{shapeType=Capsule, length=0.5, radius=0.5, shapeName=dyn_grenadeBk_l_MSTR_capsule}]
 dyn_grenadeBk_r_01 [{shapeType=Capsule, length=3, radius=3, shapeName=dyn_grenadeBk_r_01_capsule, center=<0 0 -3.6>, rotation=<rot 90 0 0>}]
 dyn_grenadeBk_r_MSTR [{shapeType=Capsule, length=0.5, radius=0.5, shapeName=dyn_grenadeBk_r_MSTR_capsule}]
 dyn_grenadeFr_l_01 [{shapeType=Capsule, length=3, radius=3, shapeName=dyn_grenadeFr_l_01_capsule, center=<0 0 3.6>, rotation=<rot 90 0 0>}]
 dyn_grenadeFr_l_MSTR [{shapeType=Capsule, length=0.5, radius=0.5, shapeName=dyn_grenadeFr_l_MSTR_capsule}]
 dyn_grenadeFr_r_01 [{shapeType=Capsule, length=3, radius=3, shapeName=dyn_grenadeFr_r_01_capsule, center=<0 0 -3.6>, rotation=<rot 90 0 0>}]
 dyn_grenadeFr_r_MSTR [{shapeType=Capsule, length=0.5, radius=0.5, shapeName=dyn_grenadeFr_r_MSTR_capsule}]
 pelvis [{shapeType=Capsule, length=10.5733, radius=13.4995, center=<10.1992 -2.9913 -0.2033>, rotation=<rot 0 -95.4352 -89.7072>}]
}
constraints [
 {
  bone1Name=dyn_grenadeBk_l_MSTR
  bone2Name=pelvis
  swing1LimitDegrees=45
  swing1Motion=Free
  swing2LimitDegrees=45
  swing2Motion=Free
  twistLimitDegrees=45
  twistMotion=Free
 }
 {bone1Name=dyn_grenadeBk_l_01, bone2Name=dyn_grenadeBk_l_MSTR, swing1LimitDegrees=5, swing2LimitDegrees=0, twistLimitDegrees=20}
 {
  bone1Name=dyn_grenadeBk_r_MSTR
  bone2Name=pelvis
  swing1LimitDegrees=45
  swing1Motion=Free
  swing2LimitDegrees=45
  swing2Motion=Free
  twistLimitDegrees=45
  twistMotion=Free
 }
 {bone1Name=dyn_grenadeBk_r_01, bone2Name=dyn_grenadeBk_r_MSTR, swing1LimitDegrees=5, swing2LimitDegrees=0, twistLimitDegrees=20}
 {
  bone1Name=dyn_grenadeFr_l_MSTR
  bone2Name=pelvis
  swing1LimitDegrees=45
  swing1Motion=Free
  swing2LimitDegrees=45
  swing2Motion=Free
  twistLimitDegrees=45
  twistMotion=Free
 }
 {bone1Name=dyn_grenadeFr_l_01, bone2Name=dyn_grenadeFr_l_MSTR, swing1LimitDegrees=5, swing2LimitDegrees=0, twistLimitDegrees=20}
 {
  bone1Name=dyn_grenadeFr_r_MSTR
  bone2Name=pelvis
  swing1LimitDegrees=45
  swing1Motion=Free
  swing2LimitDegrees=45
  swing2Motion=Free
  twistLimitDegrees=45
  twistMotion=Free
 }
 {bone1Name=dyn_grenadeFr_r_01, bone2Name=dyn_grenadeFr_r_MSTR, swing1LimitDegrees=5, swing2LimitDegrees=0, twistLimitDegrees=20}
]
