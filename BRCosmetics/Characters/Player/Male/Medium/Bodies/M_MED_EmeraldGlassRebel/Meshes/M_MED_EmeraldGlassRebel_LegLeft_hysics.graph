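# Physics asset for the left leg of the EmeraldGlassRebel medium male player body, with capsules for the leg and its dynamic boot flap bones.
# Thigh and calf capsules cover the limb, while a tiny master capsule links a two-segment flap chain to the calf.
# The calf-to-thigh and master-to-calf joints swing and twist freely.
# Each flap segment is constrained to a narrow single-axis swing, so the flap bends like a hinge.
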PhysicsAsset /BRCosmetics/Characters/Player/Male/Medium/Bodies/M_MED_EmeraldGlassRebel/Meshes/M_MED_EmeraldGlassRebel_LegLeft_hysics
bodies {
 calf_l [{
   shapeType=Capsule
   length=29.2546
   radius=8.2602
   shapeName=calf_l_capsule
   center=<-14.342 -0.8802 0.2326>
   rotation=<rot 12.1021 -80.723 -91.86>
  }]
 dyn_bootFlap_l_01 [{shapeType=Capsule, length=5.5125, radius=1.461, shapeName=dyn_bootFlap_l_01_capsule, center=<3.7139 0 0>, rotation=<rot 90 0 0>}]
 dyn_bootFlap_l_02 [{shapeType=Capsule, length=5.5125, radius=1.461, shapeName=dyn_bootFlap_l_02_capsule, center=<3.7139 0 0>, rotation=<rot 90 0 0>}]
 dyn_bootFlap_l_MSTR [{shapeType=Capsule, length=0.505, radius=0.505, shapeName=dyn_bootFlap_l_MSTR_capsule, rotation=<rot 0 -90 0>}]
 thigh_l [{shapeType=Capsule, length=13.6835, radius=10.4985, shapeName=thigh_l_capsule, center=<-21.7647 -0.6005 -1.4089>, rotation=<rot 90 0 0>}]
}
constraints [
 {
  bone1Name=dyn_bootFlap_l_MSTR
  bone2Name=calf_l
  swing1LimitDegrees=45
  swing1Motion=Free
  swing2LimitDegrees=45
  swing2Motion=Free
  twistLimitDegrees=45
  twistMotion=Free
 }
 {bone1Name=dyn_bootFlap_l_01, bone2Name=dyn_bootFlap_l_MSTR, swing1LimitDegrees=15, swing2LimitDegrees=0, twistLimitDegrees=0}
 {bone1Name=dyn_bootFlap_l_02, bone2Name=dyn_bootFlap_l_01, swing1LimitDegrees=15, swing2LimitDegrees=0, twistLimitDegrees=0}
 {
  bone1Name=calf_l
  bone2Name=thigh_l
  swing1LimitDegrees=45
  swing1Motion=Free
  swing2LimitDegrees=45
  swing2Motion=Free
  twistLimitDegrees=45
  twistMotion=Free
 }
]
