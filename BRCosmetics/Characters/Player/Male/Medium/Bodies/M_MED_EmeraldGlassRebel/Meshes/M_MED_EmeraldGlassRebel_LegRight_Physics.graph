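# Physics asset for the right leg of the EmeraldGlassRebel male medium player cosmetic body, supporting a dangling boot flap.
# Capsule bodies cover the thigh, calf, flap master bone and two flap segments.
# Constraints chain thigh to calf to master with free rotation, while both flap segments get a narrow swing limit on one axis, forming hinge-like joints.
# Most capsule names retain left-side labels, suggesting a copy from the left leg.
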PhysicsAsset /BRCosmetics/Characters/Player/Male/Medium/Bodies/M_MED_EmeraldGlassRebel/Meshes/M_MED_EmeraldGlassRebel_LegRight_Physics
bodies {
 calf_r [{
   shapeType=Capsule
   length=30.5835
   radius=8.7672
   shapeName=calf_l_capsule
   center=<14.342 0.8802 -0.2326>
   rotation=<rot 12.1021 -80.723 88.14>
  }]
 dyn_bootFlap_r_01 [{shapeType=Capsule, length=3.5651, radius=2.4349, shapeName=dyn_bootFlap_l_01_capsule, center=<1 -1.6 0>, rotation=<rot 0 0 -180>}]
 dyn_bootFlap_r_02 [{shapeType=Capsule, length=3.5651, radius=2.4349, shapeName=dyn_bootFlap_l_02_capsule, center=<1 -1.6 0>, rotation=<rot 0 0 -180>}]
 dyn_bootFlap_r_MSTR [{shapeType=Capsule, length=0.505, radius=0.505, shapeName=dyn_bootFlap_r_MSTR_capsule, rotation=<rot 0 -90 0>}]
 thigh_r [{shapeType=Capsule, length=13.0311, radius=11.4005, shapeName=thigh_r_capsule, center=<20.9587 0.582 1.1524>, rotation=<rot 90 0 0>}]
}
constraints [
 {
  bone1Name=dyn_bootFlap_r_MSTR
  bone2Name=calf_r
  swing1LimitDegrees=45
  swing1Motion=Free
  swing2LimitDegrees=45
  swing2Motion=Free
  twistLimitDegrees=45
  twistMotion=Free
 }
 {bone1Name=dyn_bootFlap_r_01, bone2Name=dyn_bootFlap_r_MSTR, swing1LimitDegrees=15, swing2LimitDegrees=0, twistLimitDegrees=0}
 {bone1Name=dyn_bootFlap_r_02, bone2Name=dyn_bootFlap_r_01, swing1LimitDegrees=15, swing2LimitDegrees=0, twistLimitDegrees=0}
 {
  bone1Name=calf_r
  bone2Name=thigh_r
  swing1LimitDegrees=45
  swing1Motion=Free
  swing2LimitDegrees=45
  swing2Motion=Free
  twistLimitDegrees=45
  twistMotion=Free
 }
]
