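# Post-process style AnimBlueprint for the EmeraldGlassRebel cosmetic body, adding secondary motion and pose fixes.
# It copies the parent mesh pose through a deform Control Rig, then four RigidBody nodes, with grenade and collar simulations at zero alpha.
# Additive and replace bone tweaks correct hands, collar and elements during skydive float, targeting head movement, celebration emote and front end.
# The head-twist collar node is unconnected.
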AnimBlueprint /BRCosmetics/Characters/Player/Male/Medium/Bodies/M_MED_EmeraldGlassRebel/Meshes/M_MED_EmeraldGlassRebel_AnimBP
target_skeleton=&/M_MED_EmeraldGlassRebel_Skeleton
graphs [
 AnimGraph {nodes [
   "Transform (Modify) Bone - Bone: dyn_collar_l_MSTR"(Alpha=Get headTwist() -> (headTwist:float), Translation="0,0,0", Rotation="-20,10,15", Scale="1,1,1") {
    rotationMode=BMM_Additive
    rotationSpace=BCS_BoneSpace
    alphaBoolBlend {blendOption=Linear}
    alphaScaleBiasClamp {bMapRange=true, inRange {max=50, min=10}}
    boneToModify {boneName=dyn_collar_l_MSTR}
    rotation=<rot -20 10 15>
   }
   n1 = RigidBody(ComponentPose=Local To Component(LocalPose=Use cached pose 'controlRigCache'())) {
    bClampLinearTranslationLimitToRefPose=true
    bOverrideWorldGravity=true
    overridePhysicsAsset=&/M_MED_EmeraldGlassRebel_LegLeft_hysics
    simulationSpace=BaseBoneSpace
    alphaBoolBlend {blendOption=Linear}
    baseBoneRef {boneName=calf_twist_02_l}
    componentAppliedLinearAccClamp=<6000 6000 6000>
    componentLinearAccScale=<0.5 0.5 0.5>
    componentLinearVelScale=<0.5 0.5 0.5>
    overrideWorldGravity {z=-980}
   }
   n2 = RigidBody(ComponentPose=@n1) {
    bClampLinearTranslationLimitToRefPose=true
    bOverrideWorldGravity=true
    overridePhysicsAsset=&/M_MED_EmeraldGlassRebel_LegRight_Physics
    simulationSpace=BaseBoneSpace
    alphaBoolBlend {blendOption=Linear}
    baseBoneRef {boneName=calf_twist_02_r}
    componentAppliedLinearAccClamp=<6000 6000 6000>
    componentLinearAccScale=<0.5 0.5 0.5>
    componentLinearVelScale=<0.5 0.5 0.5>
    overrideWorldGravity {z=-980}
   }
   n3 = RigidBody(ComponentPose=@n2, Alpha=0) {
    bClampLinearTranslationLimitToRefPose=true
    bOverrideWorldGravity=true
    overridePhysicsAsset=&/M_MED_EmeraldGlassRebel_Grenades_Physics
    simulationSpace=BaseBoneSpace
    alphaBoolBlend {blendOption=Linear}
    baseBoneRef {boneName=pelvis}
    componentAppliedLinearAccClamp=<6000 6000 6000>
    componentLinearAccScale=<0.5 0.5 0.5>
    componentLinearVelScale=<0.5 0.5 0.5>
    overrideWorldGravity {z=-980}
   }
   n4 = RigidBody(ComponentPose=@n3, Alpha=0) {
    bClampLinearTranslationLimitToRefPose=true
    bOverrideWorldGravity=true
    overridePhysicsAsset=&/M_MED_EmeraldGlassRebel_Collar_Physics
    simulationSpace=BaseBoneSpace
    alphaBoolBlend {blendOption=Linear}
    baseBoneRef {boneName=spine_05}
    componentAppliedLinearAccClamp=<6000 6000 6000>
    componentLinearAccScale=<0.1 0.1 0.1>
    componentLinearVelScale=<0.1 0.1 0.1>
    overrideWorldGravity {z=-980}
   }
   n5 = Get bIsSkydiveFloating()
   n6 = "Transform (Modify) Bone - Bone: hand_l"!(ComponentPose=@n4, bAlphaBoolEnabled:bool=@n5, Translation="0,0,0", Rotation="8,-60,0", Scale="1,1,1") -> (Pose) {
    alphaInputType=Bool
    rotationMode=BMM_Replace
    rotationSpace=BCS_BoneSpace
    alphaBoolBlend {blendOption=Linear}
    boneToModify {boneName=hand_l}
    rotation {pitch=8, yaw=-60}
   }
   n7 = "Transform (Modify) Bone - Bone: hand_r"!(ComponentPose=@n6, bAlphaBoolEnabled:bool=@n5, Translation="0,0,0", Rotation="12,-60,0", Scale="1,1,1") -> (Pose) {
    alphaInputType=Bool
    rotationMode=BMM_Replace
    rotationSpace=BCS_BoneSpace
    alphaBoolBlend {blendOption=Linear}
    boneToModify {boneName=hand_r}
    rotation {pitch=12, yaw=-60}
   }
   n8 = Get headDn() -> (headDn:float)
   n9 = "Transform (Modify) Bone - Bone: dyn_collarFr_c_MSTR"(ComponentPose=@n7, Alpha=@n8, Translation="0,0,0", Rotation="-20,-30,0", Scale="1,1,1") {
    rotationMode=BMM_Additive
    rotationSpace=BCS_BoneSpace
    alphaBoolBlend {blendOption=Linear}
    boneToModify {boneName=dyn_collarFr_c_MSTR}
    rotation {pitch=-20, yaw=-30}
   }
   n10 = "Transform (Modify) Bone - Bone: dyn_collar_r_MSTR"(ComponentPose=@n9, Alpha=@n8, Translation="0,0,0", Rotation="0,0,10", Scale="1,1,1") {
    rotationMode=BMM_Additive
    rotationSpace=BCS_BoneSpace
    alphaBoolBlend {blendOption=Linear}
    boneToModify {boneName=dyn_collar_r_MSTR}
    rotation {roll=10}
   }
   n11 = Get isCelebrating() -> (isCelebrating:bool)
   n12 = "Transform (Modify) Bone - Bone: dyn_element_l_01"!(ComponentPose=@n10, bAlphaBoolEnabled:bool=@n11, Translation="0,0,0", Rotation="15,5,15", Scale="1,1,1") -> (Pose) {
    alphaInputType=Bool
    rotationMode=BMM_Additive
    rotationSpace=BCS_BoneSpace
    alphaBoolBlend {blendOption=Linear}
    boneToModify {boneName=dyn_element_l_01}
    rotation=<rot 15 5 15>
   }
   n13 = "Transform (Modify) Bone - Bone: dyn_element_r_01"!(ComponentPose=@n12, bAlphaBoolEnabled:bool=@n11, Translation="0,0,0", Rotation="8,-2,30", Scale="1,1,1") -> (Pose) {
    alphaInputType=Bool
    rotationMode=BMM_Additive
    rotationSpace=BCS_BoneSpace
    alphaBoolBlend {blendOption=Linear}
    boneToModify {boneName=dyn_element_r_01}
    rotation=<rot 8 -2 30>
   }
   n14 = Get bIsInFrontEnd()
   n15 = "Transform (Modify) Bone - Bone: dyn_collar_l_MSTR"!(ComponentPose=@n13, bAlphaBoolEnabled:bool=@n14, Translation="0,0,0", Rotation="5,0,0", Scale="1,1,1") -> (Pose) {
    alphaInputType=Bool
    rotationMode=BMM_Additive
    rotationSpace=BCS_BoneSpace
    alphaBoolBlend {blendOption=Linear}
    boneToModify {boneName=dyn_collar_l_MSTR}
    rotation {pitch=5}
   }
   n16 = "Transform (Modify) Bone - Bone: dyn_collarBk_c_MSTR"!(ComponentPose=@n15, bAlphaBoolEnabled:bool=@n14, Translation="0,0,0", Rotation="0,-10,0", Scale="1,1,1") -> (Pose) {
    alphaInputType=Bool
    rotationMode=BMM_Additive
    rotationSpace=BCS_BoneSpace
    alphaBoolBlend {blendOption=Linear}
    boneToModify {boneName=dyn_collarBk_c_MSTR}
    rotation {yaw=-10}
   }
   Output Pose(Result=Component To Local(ComponentPose=@n16)) {layerGroup=None}
   n17 = Control Rig(Source=Copy Pose From Mesh() {bUseAttachedParent=true}) {
    alphaBoolBlend {blendOption=Linear}
    controlRigAssetReference {blueprintRigClass=&/M_MED_DeformRig_EmeraldGreenRebel_ControlRig_new.M_MED_DeformRig_EmeraldGreenRebel_ControlRig_new_C}
   }
   Save cached pose 'controlRigCache'(Pose=@n17) {cacheName=controlRigCache}
  ]}
 EventGraph {nodes [
   Event BlueprintUpdateAnimation()
   > n1 = Sequence!(execute) -> (then_0, then_1, then_2:Exec, then_3:Exec)
   > Set bodyMesh(bodyMesh:Skeletal Mesh Component Object Reference=GetOwnerSkeletalMeshComponent()) -> (Output_Get:Skeletal Mesh Component Object Reference)
   Set headTwist(
    execute=@n1.then_1
    headTwist:float=GetAlphaFromReferenceTransformDelta(
     SkeletalMeshComponent=Get bodyMesh() -> (bodyMesh:Skeletal Mesh Component Object Reference)
     BoneName=head
     BaseName=spine_05
     DeltaComponentType=QuaternionTwist
     InputRange=(X=15,Y=50)
     TwistAxis="1,0,0"
    ).RawOutput
   ) -> (Output_Get:float)
   n2 = Branch(execute=@n1.then_2, Condition=Get bIsTargeting())
   > Set headDn(headDn:float=GetAlphaFromReferenceTransformDelta(SkeletalMeshComponent=Get bodyMesh(), BoneName=head, BaseName=spine_05, InputRange=(X=0,Y=20), TwistAxis="1,0,0")) -> (Output_Get:float)
   Set headDn(execute=@n2.else, headDn=0)
   n3 = GetAnimInstance(self=GetOwnerSkeletalMeshComponent())
   Is Valid(exec=@n1.then_3, InputObject=@n3)
   > n4 = Branch(Condition="Equal (Object)"(
     A=GetCurrentActiveMontage!(self:Anim Instance Object Reference=@n3) -> (ReturnValue:Anim Montage Object Reference)
     B=/Game/Animation/Game/MainPlayer/Emotes/Emote_Celebration_M
    ))
   > Set isCelebrating(isCelebrating:bool=true) -> (Output_Get:bool)
   Set isCelebrating(execute=@n4.else, isCelebrating=false)
  ]}
]
variables [
 bodyMesh {type=SkeletalMeshComponent}
 headDn {type=number}
 headTwist {type=number}
 isCelebrating {type=boolean}
]
