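# Physics asset for the EmeraldGlassRebel medium male character's face accessory, covering two dynamic antler bones and the upper body around them.
# Single capsule bodies sit on each antler, the head, both neck bones, two upper spine bones and both clavicles.
# Each antler constrains to the head with zero swing limits and a twist limit, allowing motion around one axis.
# The head, neck, spine and clavicle chain uses free angular constraints.
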PhysicsAsset /BRCosmetics/Characters/Player/Male/Medium/Bodies/M_MED_EmeraldGlassRebel/Meshes/Parts/M_MED_EmeraldGlassRebel_FaceAcc_Physics
bodies {
 clavicle_l [{
   shapeType=Capsule
   length=17.743
   radius=10.2498
   shapeName=clavicle_l_capsule
   center=<5.3618 -2.6036 -1.8438>
   rotation=<rot -0.6798 -82.9453 95.4752>
  }]
 clavicle_r [{
   shapeType=Capsule
   length=17.743
   radius=10.2498
   shapeName=clavicle_l_capsule
   center=<-5.3618 2.6036 1.8438>
   rotation=<rot -0.6798 -82.9453 -84.5248>
  }]
 dyn_antler_l_01 [{shapeType=Capsule, length=28.8109, radius=7.292, center=<0 -19.4496 0>}]
 dyn_antler_r_01 [{shapeType=Capsule, length=28.8109, radius=7.292, center=<0 -19.4496 0>}]
 head [{shapeType=Capsule, length=5.5449, radius=10.2482, center=<11.2051 2.0033 0>, rotation=<rot 90 0 0>}]
 neck_01 [{shapeType=Capsule, length=12.7467, radius=8.982, shapeName=neck_01_capsule, rotation=<rot 0 -90 0>}]
 neck_02 [{shapeType=Capsule, length=0, radius=7.8868, shapeName=neck_02_capsule, rotation=<rot 0 -90 0>}]
 spine_04 [{shapeType=Capsule, length=0, radius=13.7956, shapeName=spine_04_capsule, rotation=<rot 0 -90 0>}]
 spine_05 [{shapeType=Capsule, length=0, radius=13.7956, shapeName=spine_05_capsule, rotation=<rot 0 -90 0>}]
}
constraints [
 {bone1Name=dyn_antler_r_01, bone2Name=head, swing1LimitDegrees=0, swing2LimitDegrees=0, twistLimitDegrees=35}
 {bone1Name=dyn_antler_l_01, bone2Name=head, swing1LimitDegrees=0, swing2LimitDegrees=0, twistLimitDegrees=35}
 {
  bone1Name=head
  bone2Name=neck_02
  swing1LimitDegrees=45
  swing1Motion=Free
  swing2LimitDegrees=45
  swing2Motion=Free
  twistLimitDegrees=45
  twistMotion=Free
 }
 {
  bone1Name=neck_01
  bone2Name=spine_05
  swing1LimitDegrees=45
  swing1Motion=Free
  swing2LimitDegrees=45
  swing2Motion=Free
  twistLimitDegrees=45
  twistMotion=Free
 }
 {
  bone1Name=neck_02
  bone2Name=neck_01
  swing1LimitDegrees=45
  swing1Motion=Free
  swing2LimitDegrees=45
  swing2Motion=Free
  twistLimitDegrees=45
  twistMotion=Free
 }
 {
  bone1Name=spine_05
  bone2Name=spine_04
  swing1LimitDegrees=45
  swing1Motion=Free
  swing2LimitDegrees=45
  swing2Motion=Free
  twistLimitDegrees=45
  twistMotion=Free
 }
 {
  bone1Name=clavicle_l
  bone2Name=spine_05
  swing1LimitDegrees=45
  swing1Motion=Free
  swing2LimitDegrees=45
  swing2Motion=Free
  twistLimitDegrees=45
  twistMotion=Free
 }
 {
  bone1Name=clavicle_r
  bone2Name=spine_05
  swing1LimitDegrees=45
  swing1Motion=Free
  swing2LimitDegrees=45
  swing2Motion=Free
  twistLimitDegrees=45
  twistMotion=Free
 }
]
