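# Animation Blueprint for the Emerald Glass Rebel face accessory, posing antler and back-hair bones on the player character.
# It copies the parent mesh pose, layering leg bones from the body mesh, then applies additive bone rotations and offsets.
# Skydive floating and head-versus-spine rotation drive the antlers; head rotation and a cosmetic variant check adjust the back hair.
# Control Rig and RigidBody nodes sit at zero alpha, and computed headTwist is unused.
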AnimBlueprint /BRCosmetics/Characters/Player/Male/Medium/Bodies/M_MED_EmeraldGlassRebel/Meshes/Parts/M_MED_EmeraldGlassRebel_FaceAcc_AnimBP
target_skeleton=&/M_MED_EmeraldGlassRebel_FaceAcc_Skeleton
graphs [
 AnimGraph {nodes [
   n1 = Layered blend per bone(
    BasePose=Copy Pose From Mesh() {bCopyCurves=true, bUseAttachedParent=true}
    BlendPoses_0=Copy Pose From Mesh(SourceMeshComponent=Get Body_Mesh())
   ) {layerSetup [{branchFilters [{boneName=thigh_l}, {boneName=thigh_r}, {boneName=calf_l}, {boneName=calf_r}]}]}
   n2 = Control Rig(Source=@n1, Alpha=0) {
    alphaBoolBlend {blendOption=Linear}
    controlRigAssetReference {blueprintRigClass=&/M_MED_EmeraldGlassRebel_FaceAcc_CtrlRig.M_MED_EmeraldGlassRebel_FaceAcc_CtrlRig_C}
   }
   n3 = Get bIsSkydiveFloating()
   n4 = "Transform (Modify) Bone - Bone: dyn_antler_r_01"!(ComponentPose=Local To Component(LocalPose=@n2), bAlphaBoolEnabled:bool=@n3, Translation="0,0,0", Rotation="0,0,15", Scale="1,1,1") -> (Pose) {
    alphaInputType=Bool
    rotationMode=BMM_Additive
    rotationSpace=BCS_BoneSpace
    alphaBoolBlend {blendOption=Linear}
    boneToModify {boneName=dyn_antler_r_01}
    rotation {roll=15}
   }
   n5 = "Transform (Modify) Bone - Bone: dyn_antler_l_01"!(ComponentPose=@n4, bAlphaBoolEnabled:bool=@n3, Translation="0,0,0", Rotation="0,0,15", Scale="1,1,1") -> (Pose) {
    alphaInputType=Bool
    rotationMode=BMM_Additive
    rotationSpace=BCS_BoneSpace
    alphaBoolBlend {blendOption=Linear}
    boneToModify {boneName=dyn_antler_l_01}
    rotation {roll=15}
   }
   n6 = RigidBody(ComponentPose=@n5, Alpha=0) {
    bClampLinearTranslationLimitToRefPose=true
    bOverrideWorldGravity=true
    overridePhysicsAsset=&/M_MED_EmeraldGlassRebel_FaceAcc_Physics
    simulationSpace=BaseBoneSpace
    alphaBoolBlend {blendOption=Linear}
    baseBoneRef {boneName=spine_05}
    componentAppliedLinearAccClamp=<6000 6000 6000>
    componentLinearAccScale=<0.1 0.1 0.1>
    componentLinearVelScale=<0.1 0.1 0.1>
    overrideWorldGravity {z=-980}
   }
   n7 = Get headRotBack() -> (headRotBack:double)
   n8 = "Transform (Modify) Bone - Bone: dyn_antler_l_01"!(ComponentPose=@n6, Alpha:float=@n7, Translation="0,0,0", Rotation="0,0,20", Scale="1,1,1") -> (Pose) {
    rotationMode=BMM_Additive
    rotationSpace=BCS_BoneSpace
    alphaBoolBlend {blendOption=Linear}
    alphaScaleBiasClamp {interpSpeedDecreasing=5, interpSpeedIncreasing=5}
    boneToModify {boneName=dyn_antler_l_01}
    rotation {roll=20}
   }
   n9 = "Transform (Modify) Bone - Bone: dyn_antler_r_01"!(ComponentPose=@n8, Alpha:float=@n7, Translation="0,0,0", Rotation="0,0,20", Scale="1,1,1") -> (Pose) {
    rotationMode=BMM_Additive
    rotationSpace=BCS_BoneSpace
    alphaBoolBlend {blendOption=Linear}
    alphaScaleBiasClamp {interpSpeedDecreasing=5, interpSpeedIncreasing=5}
    boneToModify {boneName=dyn_antler_r_01}
    rotation {roll=20}
   }
   n10 = "Transform (Modify) Bone - Bone: dyn_hairBk_c_01"(ComponentPose=@n9, Alpha=@n7, Translation="0,5,2", Rotation="0,0,-25", Scale="1,1,1") {
    rotationMode=BMM_Additive
    rotationSpace=BCS_BoneSpace
    translationMode=BMM_Additive
    translationSpace=BCS_BoneSpace
    alphaBoolBlend {blendOption=Linear}
    boneToModify {boneName=dyn_hairBk_c_01}
    rotation {roll=-25}
    translation {y=5, z=2}
   }
   n11 = "Transform (Modify) Bone - Bone: dyn_hairBk_c_01"!(
    ComponentPose=@n10
    bAlphaBoolEnabled:bool=Get isCollarVariant() -> (isCollarVariant:bool)
    Translation="0,0,2"
    Rotation="0,0,-25"
    Scale="1,1,1"
   ) -> (Pose) {
    alphaInputType=Bool
    rotationMode=BMM_Additive
    rotationSpace=BCS_BoneSpace
    translationMode=BMM_Additive
    translationSpace=BCS_BoneSpace
    alphaBoolBlend {blendOption=Linear}
    boneToModify {boneName=dyn_hairBk_c_01}
    rotation {roll=-25}
    translation {z=2}
   }
   Output Pose(Result=Component To Local(ComponentPose=@n11)) {layerGroup=None}
   Get headTwist() -> (headTwist:float)
  ]}
 EventGraph {nodes [
   Event BlueprintUpdateAnimation()
   > n1 = Sequence!(execute) -> (then_0, then_1, then_2:Exec)
   > Set headTwist(headTwist:float=GetAlphaFromReferenceTransformDelta(
     SkeletalMeshComponent=Get Body_Mesh()
     BoneName=head
     BaseName=spine_05
     DeltaComponentType=QuaternionTwist
     InputRange=(X=15,Y=50)
     TwistAxis="1,0,0"
    ).RawOutput) -> (Output_Get:float)
   Set headRotBack(
    execute=@n1.then_1
    headRotBack:double=GetAlphaFromReferenceTransformDelta(SkeletalMeshComponent=Get Body_Mesh(), BoneName=head, BaseName=spine_05, InputRange=(X=-40,Y=0), OutputRange=(X=1,Y=0), TwistAxis="1,0,0")
   ) -> (Output_Get:double)
   n2 = Cast To FortPlayerPawn(execute=@n1.then_2, Object=GetOwningActor())
   > n3 = GetVariantChannelActiveIndexForItemActor!(
    execute
    Item:Athena Cosmetic Item Definition Object Reference=/BRCosmetics/Athena/Items/Cosmetics/Characters/Character_EmeraldGlassRebel
    Actor:Actor Object Reference=@n2
    Channel:Gameplay Tag Structure=(TagName="Cosmetics.Variant.Channel.Parts")
    VariantList:Array of Gameplay Tag Structures=Make Array("[0]"=(TagName="Cosmetics.Variant.Property.Stage1"))
    bErrorIfChannelMissing:bool=false
    bErrorIfTagNotInArray:bool=false
   ) -> (then, Index:int=0, bIndexIsValid:bool, bFoundChannel:bool)
   > n4 = Branch(Condition=AND Boolean(A="Equal (Integer)"(A=@n3), B=@n3.bIndexIsValid))
   > Set isCollarVariant(isCollarVariant:bool=true) -> (Output_Get:bool)
   Set isCollarVariant(execute=@n4.else, isCollarVariant=false)
   Event OnCharacterCustomizationLoaded()
   > n5 = Cast To PlayerPawn_Athena(Object=TryGetPawnOwner())
   > Set Body_Mesh(Body_Mesh=GetSkeletalMeshForPartType(self=@n5))
  ]}
]
variables [
 Body_Mesh {type=SkeletalMeshComponent}
 headRotBack {type=number}
 headTwist {type=number}
 isCollarVariant {type=boolean}
]
